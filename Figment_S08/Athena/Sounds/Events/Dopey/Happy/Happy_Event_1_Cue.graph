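# Sound cue for the Dopey Happy event that plays a small happy explosion sound, routed to the map-wide events sound class.
# Four equally weighted wave variants feed a random node, then a default Modulator for slight variation, a Snow-folder attenuation asset without air absorption, and a fixed delay at the root.
# Overridden concurrency allows one instance at a time and rejects new plays, and the cue's volume is boosted.
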SoundCue /Figment_S08/Athena/Sounds/Events/Dopey/Happy/Happy_Event_1_Cue
root=Delay_0
nodes [
 {
  id=Attenuation_0
  children [Modulator_1]
  properties {attenuationSettings=/Figment_S08/Athena/Sounds/Events/Dopey/Snow/Snow_MLTI_Close_NoAirAbsorp_Att}
 }
 {id=Delay_0, children [Attenuation_0], properties {delayMax=1.45, delayMin=1.45}}
 {id=Modulator_1, children [Random_0]}
 {id=Random_0, children [WavePlayer_0,WavePlayer_1,WavePlayer_2,WavePlayer_3], properties {weights [1,1,1,1]}}
 {id=WavePlayer_0, properties {soundWaveAssetPtr=&/Explo_Happy_Small_01}}
 {id=WavePlayer_1, properties {soundWaveAssetPtr=&/Explo_Happy_Small_02}}
 {id=WavePlayer_2, properties {soundWaveAssetPtr=&/Explo_Happy_Small_03}}
 {id=WavePlayer_3, properties {soundWaveAssetPtr=&/Explo_Happy_Small_04}}
]
properties {
 bOverrideConcurrency=true
 maxDistance=22500
 soundClassObject=/Figment_S08/Athena/Sounds/Events/SoundClasses/Fort_Events_Mapwide
 volumeMultiplier=2
 attenuationOverrides {falloffDistance=3000, attenuationShapeExtents {x=0}}
 concurrencyOverrides {maxCount=1, resolutionRule=PreventNew}
}
