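# Sound cue for a "Happy" explosion in the Dopey event, routed to the mapwide events sound class.
# After a fixed half-second delay, a mixer layers two random picks of three waves each.
# The close explosion layer passes through a modulator into close attenuation, while the distant variants use their own distant attenuation.
# Overridden concurrency caps playback at four instances, stopping the oldest.
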SoundCue /Figment_S08/Athena/Sounds/Events/Dopey/Happy/Happy_Event_2_Cue
root=Delay_0
nodes [
 {
  id=Attenuation_0
  children [Modulator_1]
  properties {attenuationSettings=/Figment_S08/Athena/Sounds/Events/Dopey/Snow/Snow_MLTI_Close_NoAirAbsorp_Att}
 }
 {
  id=Attenuation_1
  children [Random_1]
  properties {attenuationSettings=/Figment_S08/Athena/Sounds/Events/Dopey/Snow/ATT_Snow_MLTI_Distant_Att}
 }
 {id=Delay_0, children [Mixer_0], properties {delayMax=0.5, delayMin=0.5}}
 {id=Mixer_0, children [Attenuation_0, Attenuation_1], properties {inputVolume [1,1]}}
 {id=Modulator_1, children [Random_0]}
 {id=Random_0, children [WavePlayer_0,WavePlayer_1,WavePlayer_2], properties {weights [1,1,1]}}
 {id=Random_1, children [WavePlayer_3,WavePlayer_4,WavePlayer_5], properties {weights [1,1,1]}}
 {id=WavePlayer_0, properties {soundWaveAssetPtr=&/Explo_Happy_Med_03}}
 {id=WavePlayer_1, properties {soundWaveAssetPtr=&/Explo_Happy_Med_01}}
 {id=WavePlayer_2, properties {soundWaveAssetPtr=&/Explo_Happy_Med_02}}
 {id=WavePlayer_3, properties {soundWaveAssetPtr=&/Explo_Happy_Med_02_Dist}}
 {id=WavePlayer_4, properties {soundWaveAssetPtr=&/Explo_Happy_Med_03_Dist}}
 {id=WavePlayer_5, properties {soundWaveAssetPtr=&/Explo_Happy_Med_01_Dist}}
]
properties {
 bOverrideConcurrency=true
 maxDistance=2097152
 soundClassObject=/Figment_S08/Athena/Sounds/Events/SoundClasses/Fort_Events_Mapwide
 volumeMultiplier=2
 attenuationOverrides {falloffDistance=3000, attenuationShapeExtents {x=0}}
 concurrencyOverrides {maxCount=4, resolutionRule=StopOldest}
}
