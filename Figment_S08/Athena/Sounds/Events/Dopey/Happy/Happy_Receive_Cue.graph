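# Sound cue for a "Happy Receive" effect in the Figment S08 Dopey event, playing one of three explosion-named wave variants.
# An equally weighted random pick runs through a modulator, a close Snow attenuation, a mixer with one empty input, and a root delay.
# The cue uses the map-wide events sound class, boosts volume, and overrides concurrency to stop the oldest instance beyond four.
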SoundCue /Figment_S08/Athena/Sounds/Events/Dopey/Happy/Happy_Receive_Cue
root=Delay_0
nodes [
 {
  id=Attenuation_0
  children [Modulator_1]
  properties {attenuationSettings=/Figment_S08/Athena/Sounds/Events/Dopey/Snow/ATT_Snow_MLTI_Close_Att}
 }
 {id=Delay_0, children [Mixer_0]}
 {id=Mixer_0, children [Attenuation_0, null], properties {inputVolume [1,1]}}
 {id=Modulator_1, children [Random_0]}
 {id=Random_0, children [WavePlayer_0,WavePlayer_1,WavePlayer_2], properties {weights [1,1,1]}}
 {id=WavePlayer_0, properties {soundWaveAssetPtr=&/Explo_Happy_Receive_04}}
 {id=WavePlayer_1, properties {soundWaveAssetPtr=&/Explo_Happy_Receive_02}}
 {id=WavePlayer_2, properties {soundWaveAssetPtr=&/Explo_Happy_Receive_03}}
]
properties {
 bOverrideConcurrency=true
 maxDistance=22500
 soundClassObject=/Figment_S08/Athena/Sounds/Events/SoundClasses/Fort_Events_Mapwide
 volumeMultiplier=2
 attenuationOverrides {falloffDistance=3000, attenuationShapeExtents {x=0}}
 concurrencyOverrides {maxCount=4, resolutionRule=StopOldest}
}
